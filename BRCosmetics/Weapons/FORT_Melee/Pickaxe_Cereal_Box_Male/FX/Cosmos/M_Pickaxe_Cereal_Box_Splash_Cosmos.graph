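# Masked two-sided lit material for Niagara sprite and mesh particles: a cosmos-themed water splash effect for the cereal box pickaxe.
# A radial flowmap warps the splash normal and mask UVs via dynamic parameter CenterWarp.
# Camera-vector triplanar nebula and star layers, shaded by the mask, blend with particle color or a two-color fallback for emissive.
# Particle alpha erodes the mask through a power curve, and MeshWPOBend drives a UV-weighted Z-axis bend.
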Material /BRCosmetics/Weapons/FORT_Melee/Pickaxe_Cereal_Box_Male/FX/Cosmos/M_Pickaxe_Cereal_Box_Splash_Cosmos
graph {
 properties {
  bUsedWithNiagaraMeshParticles=true
  bUsedWithNiagaraSprites=true
  blendMode=BLEND_Masked
  pixelDepthOffsetMode=PDOM_Legacy
  twoSided=true
  usedShadingModels=Default Lit
 }
 nodes [
  Constant() {r=0.2}
  Constant()
  n1 = VectorParameter() {parameterName=Tile, defaultValue=<rgba 1 1 0 0>}
  n2 = MaterialFunctionCall(Texture Scale=Constant() {r=1.4}) {materialFunction=&ScaleUVsByCenter}
  n3 = Subtract(A=ComponentMask(None=TextureSample(UVs=@n2.UVs) {samplerType=SAMPLERTYPE_LinearColor, texture=/Game/Environments/Hexmap/HexClouds/T_RadialFlowmap01}.RGB) {g=true, r=true}) {constB=0.5}
  n4 = DynamicParameter() {parameterIndex=1, defaultValue=<rgba 0 0 1 1>, paramNames [CenterWarp,MeshWPOBend,Param3,Param4]}
  n5 = Multiply(A=AppendVector(A=@n1.R, B=@n1.R), B=Add(A=TextureCoordinate(), B=Multiply(A=@n3, B=@n4.CenterWarp)))
  LinearInterpolate(
   A=TextureSampleParameter2D(UVs=@n5) {parameterName=Normal, samplerType=SAMPLERTYPE_Normal, texture=/Game/Effects/Fort_Effects/Textures/Water/T_water_splash_04_n}.RGB
   B=Constant3Vector() {constant=<rgba 0 0 1 0>}
   Alpha=ScalarParameter() {parameterName=Flatten Normal}
  )
  n6 = TextureSampleParameter2D(UVs=@n5) {parameterName=Mask, texture=/Game/Effects/Fort_Effects/Textures/Water/T_water_splash_04}
  n7 = Saturate(None=Add(A=ScalarParameter() {defaultValue=0.4, group=Galaxy, parameterName=Background Contrast Add}, B=@n6.B))
  n8 = ParticleColor()
  n9 = Multiply(
   A=LinearInterpolate(
    A=VectorParameter() {parameterName=Color A, defaultValue=<rgba 2 0.0667 0 1>}.RGB
    B=VectorParameter() {parameterName=Color B, defaultValue=<rgba 1.5 0.4875 0.15 1>}.RGB
    Alpha=@n7
   )
   B=@n8.RGB
  )
  n10 = Time()
  n11 = LinearInterpolate(A=@n10, B=Multiply(A=@n10, B=Constant()), Alpha=ScalarParameter() {group=Galaxy, parameterName=Background Rotation})
  n12 = AppendVector(
   A=AppendVector(
    A=ScalarParameter() {defaultValue=0.02, group=Galaxy, parameterName=Galaxy Rotation Speed}
    B=ScalarParameter() {defaultValue=-0.02, group=Galaxy, parameterName=Galaxy Rotation Speed_1}
   )
   B=ScalarParameter() {defaultValue=0.1, group=Galaxy, parameterName=Galaxy Rotation Speed_2}
  )
  n13 = TextureObjectParameter() {
   group=Galaxy
   parameterName=GalaxyTexture
   texture=/Game/Characters/Player/Male/Medium/Bodies/M_Med_Soldier_04/Skins/Galaxy/Textures/T_FN_Nebula
  }
  n14 = Desaturation(
   None=MaterialFunctionCall!(
    CameraVector
    Rotation Speed=Multiply(A=@n11, B=@n12)
    Texture=@n13
    Tiling=ScalarParameter() {defaultValue=0.5, group=Galaxy, parameterName=GalaxyTiling}
    Offset=DynamicParameter() {defaultValue=<rgba 0 1 1 1>, paramNames [Random, Param2, Param3, Galaxy Mix]}.Random
    AxisFadeContrast=ScalarParameter() {defaultValue=6, group=Galaxy, parameterName=GalaxyAxisFade}
   ) {materialFunction=/Game/Weapons/WeaponSkins/MaterialFunction/TriplanarCameraVector_Rotated}.Result
   Fraction=ScalarParameter() {group=Galaxy, parameterName=Galaxy Saturation}
  )
  n15 = Desaturation(None=@n14, Fraction=ScalarParameter() {defaultValue=0.5, group=Galaxy, parameterName=Star Saturation})
  n16 = Multiply(A=@n15, B=VectorParameter() {group=Galaxy, parameterName=Star Brightness, defaultValue=<rgba 300 300 300 1>}.RGB)
  n17 = Time()
  n18 = ScalarParameter() {defaultValue=-0.125, group=Galaxy, parameterName=Small Star Rotation speed}
  n19 = TextureObject() {samplerType=SAMPLERTYPE_Masks, texture=/Game/Characters/Player/Male/Medium/Bodies/M_Med_Soldier_04/Skins/Galaxy/Textures/Celestial_T_Stars}
  n20 = ComponentMask(None=MaterialFunctionCall(
    Rotation Speed=Multiply(A=@n17, B=@n18)
    Texture=@n19
    Tiling=ScalarParameter() {defaultValue=5, group=Galaxy, parameterName=Small Star Tiling}
   ) {materialFunction=&TriplanarCameraVector_Rotated}.Result) {b=true}
  n21 = Multiply(A=@n20, B=@n20)
  n22 = VectorParameter() {group=Galaxy, parameterName=Star Brightness_1, defaultValue=<rgba 950 950 950 1>}
  n23 = TextureObject() {samplerType=SAMPLERTYPE_Masks, texture=&Celestial_T_Stars}
  n24 = ComponentMask(None=MaterialFunctionCall!(
    CameraVector
    Texture=@n23
    Tiling=ScalarParameter() {defaultValue=2, group=Galaxy, parameterName=Bright Star Tiling}
    Offset
    AxisFadeContrast
   ) {materialFunction=/Engine/Functions/Engine_MaterialFunctions01/Texturing/TriplanarCameraVector}.Result) {r=true}
  n25 = Multiply(A=@n24, B=@n24)
  n26 = TextureSample(UVs=Panner(Coordinate=TextureCoordinate() {uTiling=2, vTiling=2}) {speedX=0.133, speedY=0.2}) {
   samplerSource=SSM_Wrap_WorldGroupSettings
   samplerType=SAMPLERTYPE_Masks
   texture=&Celestial_T_Stars
  }
  n27 = Multiply(
   A=LinearInterpolate(A=LinearInterpolate(A=@n14, B=@n16, Alpha=@n21), B=Multiply(A=@n15, B=@n22.RGB), Alpha=Multiply(A=@n25, B=@n26.G))
   B=@n7
  )
  n28 = VectorParameter() {group=Galaxy, parameterName=Background ColorB, defaultValue=<rgba 1 1 1 1>}
  n29 = Multiply(A=@n7, B=VectorParameter() {group=Galaxy, parameterName="Background  ColorA", defaultValue=<rgba 10 10 10 1>}.RGB) {constB=10}
  n30 = MaterialFunctionCall(In=@n7, Contrast=ScalarParameter() {defaultValue=0.2, group=Galaxy, parameterName=Background Contrast}) {materialFunction=&CheapContrast}
  n31 = LinearInterpolate(
   A=@n8.RGB
   B=LinearInterpolate(A=Multiply(A=@n27, B=@n28.RGB) {constB=10}, B=Multiply(A=@n27, B=@n29), Alpha=Saturate(None=@n30.Result)) {constAlpha=1}
   Alpha=DynamicParameter() {defaultValue=<rgba 0 1 1 1>, paramNames [Random, Param2, Param3, Galaxy Mix]}.Galaxy Mix
  )
  n32 = ScalarParameter() {defaultValue=1, group=Galaxy, parameterName=Ues Cosmos}
  MaterialFunctionCall(Emissive=Multiply(A=LinearInterpolate(A=@n9, B=@n31, Alpha=@n32), B=ScalarParameter() {defaultValue=0.15, parameterName=EmissiveAdd}) {constB=0.15}) {materialFunction=&MF_FX_EmissiveAdjustments}
  n33 = ChannelMaskParameter(None=@n6.RGB) {maskChannel=Blue, parameterName=ErodeMask, defaultValue=<rgba 0 0 1 0>}
  n34 = LinearInterpolate(Alpha=@n8.A) {constA=2, constB=-0.1}
  Multiply(A=Multiply(A=@n8.A, B=Power(Base=@n33.R, Exp=@n34)), B=ChannelMaskParameter(None=@n6.RGBA) {parameterName=OpacityMask}.R)
  n35 = ComponentMask(None=TextureCoordinate()) {g=true}
  Multiply(A=Multiply(A=Constant3Vector() {constant=<rgba 0 0 1 0>}, B=@n4.MeshWPOBend), B=Power(Base=OneMinus(None=@n35)) {constExponent=3})
  OneMinus(None=Divide(A=Time() {bOverride_Period=true, period=2})) {desc=debug erosion}
 ]
}
