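# Idle Niagara effect for the Cosmos Cereal Box pickaxe, shedding short-lived blue-to-violet mesh drops around it.
# Drops and Drops_Small pulse spawn rate over a loop through a movement-threshold module, while Drops_World spawns by distance traveled.
# All three sample the skeletal mesh, spawn on cylinder surfaces, then apply downward acceleration, noise, drag and grow-shrink curves.
# Front-end modifiers adjust camera offset on both pulsed emitters and spawn offset on Drops.
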NiagaraSystem /BRCosmetics/Weapons/FORT_Melee/Pickaxe_Cereal_Box_Male/FX/Cosmos/NS_Pickaxe_Cereal_Box_Idle_Cosmos
dependencies {usedDataInterfaces [&NiagaraDataInterfaceCurve], usedDynamicInputs [], usedModules [&SystemState]}
emitters {
 Drops {
  emitterUpdateScript [
   &EmitterState(Life Cycle Mode=::NewEnumerator1, Loop Duration=4, UseLoopDelay=false)
   &self:ThresholdSpawnRate(
    MovementThreshold=80
    SpawnRate:NiagaraFloat=&FloatFromCurve(
     FloatCurve=DataInterface() {propertyValues {Curve {keys [
         {interpMode=RCIM_Cubic, tangentMode=RCTM_User, time=0.25, value=0.5}
         {interpMode=RCIM_Cubic, tangentMode=RCTM_User, time=0.5, value=2}
         {interpMode=RCIM_Cubic, tangentMode=RCTM_User, time=0.75, value=0.5}
        ]}}}
     CurveIndex=$Emitter.NormalizedLoopAge
    )
   ) {module=ThresholdSpawnRate}
   &SpawnRate(SpawnRate=$Emitter.ThresholdSpawnRate, Use Spawn Probability=false)
  ]
  particleSpawnScript [
   &InitializeParticle(
    Color=&RandomRangeLinearColor(Minimum=<rgba 0.2917 0.15 1 0.9>, Maximum=<rgba 0.7167 0.15 1 0.9>)
    Lifetime Max=0.6
    Lifetime Min=0.5
    Lifetime Mode=::NewEnumerator1
    Mesh Scale Max=<1.2 0.5 0.5>
    Mesh Scale Min=<1.5 0.7 0.7>
    Mesh Scale Mode=::NewEnumerator2
    Sprite Rotation Angle=&RandomRangeFloat(Minimum=0, Maximum=1)
    Sprite Rotation Mode=::NewEnumerator1
    UsePositionOffset=false
    Write Material Random=true
   )
   &SkeletalMeshLocation(
    Apply To Particle Position=false
    Bone / Socket Index=&SpawnGroup()
    Inherit Velocity=true
    "Inherit Velocity (Scale)"=0.02
    MaskSpawnGroup=false
    Mesh Sampling Type=::NewEnumerator5
    Sampled Position Offset Coordinate Space:ENiagaraMeshTransforms=::NewEnumerator3
    Skeletal Mesh=$User.SkeletalMesh
    Socket Sampling Mode=::NewEnumerator5
    VelocitySpeedLimit=false
   ) {module=SkeletalMeshLocation001}
   &CylinderLocation(
    Cylinder Axis:Vector3f=<-1 0 0>
    Cylinder Height=1
    Cylinder Radius=7.5
    Hemicircle Internal Cap=false
    Hemicircle X=true
    MaskSpawnGroup=false
    Non Uniform Scale:Vector3f=<1.3 1.2 1>
    Offset=&FN_FrontEndVector3Modifier(Gameplay=<0 0 0>, Front End=<2 0 0>)
    Offset Mode:ENiagara_OffsetMode=::NewEnumerator0
    Override Seed=false
    Rotation Coordinate Space:ENiagaraCoordinateSpace=::Simulation
    Rotation Mode:ENiagara_RotationMode=::NewEnumerator1
    Surface Only=true
    Use Endcaps In Surface Only Mode=false
    UseOwnerScale=true
   )
   &AddVelocity(Scale Added Velocity=<1 1 10>, Velocity=&RandomRangeVector(Minimum=<0 2 2>, Maximum=<0 -2 2.5>))
   &AddVelocityFromPoint(
    Module.Use Falloff Distance=false
    Origin Offset=<2 0 -20>
    Velocity Strength=&RandomRangeFloat(Minimum=18, Maximum=20)
   )
  ]
  particleUpdateScript [
   &ParticleState()
   &AddVelocity(Coordinate Space=::World, Velocity=&RandomRangeVector(Minimum=<0.5 0.5 0>, Maximum=<-0.5 -0.5 0>)) {module=AddVelocity001}
   &InheritVelocity(Inherited Velocity Amount Scale=<-0.05 -0.05 0>, Inherited Velocity Speed Limit=50, Speed Limit=true, UseSpeedThreshold=false)
   &AccelerationForce(Acceleration=&RandomRangeVector(Minimum=<0 0 -15>, Maximum=<0 0 -17>))
   &VectorNoiseForce(Force Amount=&RandomRangeFloat(Minimum=10, Maximum=20), Module.Use Random Noise Time Interval=false)
   &Drag(Drag=3, Use Linear Drag=true, Use Rotational Drag=false)
   &SolveForcesAndVelocity(Clamp Velocity=false, Limit Acceleration=false)
   &GenerateLocationEvent(Module.Use Event Send Rate=true)
   &ScaleMeshSize(Scale Factor=&VectorFromFloat(Value=&FloatFromCurve(
      FloatCurve=DataInterface() {propertyValues {Curve {keys [
          {interpMode=RCIM_Cubic, tangentMode=RCTM_User, time=0, value=0}
          {interpMode=RCIM_Cubic, tangentMode=RCTM_User, time=0.45, value=1}
          {interpMode=RCIM_Cubic, tangentMode=RCTM_User, time=0.5, value=1}
          {interpMode=RCIM_Cubic, tangentMode=RCTM_User, time=0.9, value=0}
         ]}}}
      Scale Curve=&RandomRangeFloat(Minimum=1.2, Maximum=1.8)
     )))
   &CameraOffset(Camera Offset Amount=&FN_FrontEndFloatModifier(
     Gameplay=&RandomRangeFloat(Minimum=5, Maximum=4)
     Front End=&RandomRangeFloat(Minimum=1, Maximum=1.5)
    ))
  ]
  rendererClasses [&NiagaraMeshRendererProperties]
  renderers [{rendererClass=&NiagaraMeshRendererProperties, rendererIndex=0}]
 }
 Drops_Small {
  emitterUpdateScript [
   &EmitterState(Life Cycle Mode=::NewEnumerator1, Loop Duration=4, UseLoopDelay=false)
   &self:ThresholdSpawnRate(
    MovementThreshold=80
    SpawnRate:NiagaraFloat=&FloatFromCurve(
     FloatCurve=DataInterface() {propertyValues {Curve {keys [
         {interpMode=RCIM_Cubic, tangentMode=RCTM_User, time=0.25, value=1}
         {interpMode=RCIM_Cubic, tangentMode=RCTM_User, time=0.5, value=3}
         {interpMode=RCIM_Cubic, tangentMode=RCTM_User, time=0.75, value=1}
        ]}}}
     CurveIndex=$Emitter.NormalizedLoopAge
    )
   ) {module=ThresholdSpawnRate}
   &SpawnRate(SpawnRate=$Emitter.ThresholdSpawnRate, Use Spawn Probability=false)
  ]
  particleSpawnScript [
   &InitializeParticle(
    Color=&RandomRangeLinearColor(Minimum=<rgba 0.2917 0.15 1 0.9>, Maximum=<rgba 0.7167 0.15 1 0.9>)
    Lifetime Max=0.4
    Lifetime Min=0.6
    Lifetime Mode=::NewEnumerator1
    Mesh Scale Max=<0.4 0.3 0.3>
    Mesh Scale Min=<0.6 0.4 0.4>
    Mesh Scale Mode=::NewEnumerator2
    Sprite Rotation Angle=&RandomRangeFloat(Minimum=0, Maximum=1)
    Sprite Rotation Mode=::NewEnumerator1
    UsePositionOffset=false
    Write Material Random=true
   )
   &SkeletalMeshLocation(
    Apply To Particle Position=false
    Bone / Socket Index=&SpawnGroup()
    Inherit Velocity=true
    "Inherit Velocity (Scale)"=0.02
    MaskSpawnGroup=false
    Mesh Sampling Type=::NewEnumerator5
    Sampled Position Offset=&RandomRangeVector(Minimum=<0 10 -10>, Maximum=<0 -10 20>)
    Sampled Position Offset Coordinate Space:ENiagaraMeshTransforms=::NewEnumerator3
    Skeletal Mesh=$User.SkeletalMesh
    Socket Sampling Mode=::NewEnumerator5
    VelocitySpeedLimit=false
   ) {module=SkeletalMeshLocation001}
   &CylinderLocation(
    Cylinder Axis:Vector3f=<-1 0 0>
    Cylinder Height=1
    Cylinder Radius=10
    Hemicircle Internal Cap=false
    Hemicircle X=true
    MaskSpawnGroup=false
    Non Uniform Scale:Vector3f=<1.5 1 1>
    Offset=&RandomRangeVector(Minimum=<1 0 -5>, Maximum=<1 0 -5>)
    Offset Mode:ENiagara_OffsetMode=::NewEnumerator0
    Override Seed=false
    Rotation Coordinate Space:ENiagaraCoordinateSpace=::Simulation
    Rotation Mode:ENiagara_RotationMode=::NewEnumerator1
    Surface Only=true
    Use Endcaps In Surface Only Mode=false
    UseOwnerScale=true
   )
   &AddVelocity(Scale Added Velocity=<1 1 10>, Velocity=&RandomRangeVector(Minimum=<0 1 5>, Maximum=<0 -1 6>))
   &AddVelocityFromPoint(
    Module.Use Falloff Distance=false
    Origin Offset=<1 0 -20>
    Velocity Strength=&RandomRangeFloat(Minimum=15, Maximum=13)
   )
  ]
  particleUpdateScript [
   &ParticleState()
   &AccelerationForce(Acceleration=&RandomRangeVector(Minimum=<0 0 -80>, Maximum=<0 0 -100>))
   &InheritVelocity(Inherited Velocity Amount Scale=<-0.1 -0.1 0>, Inherited Velocity Speed Limit=20, Speed Limit=true, UseSpeedThreshold=true)
   &VectorNoiseForce(Force Amount=&RandomRangeFloat(Minimum=250, Maximum=150), Module.Use Random Noise Time Interval=false)
   &Drag(Drag=5, Use Linear Drag=true, Use Rotational Drag=false)
   &SolveForcesAndVelocity(Clamp Velocity=false, Limit Acceleration=false)
   &GenerateLocationEvent(Module.Use Event Send Rate=true)
   &ScaleMeshSize(Scale Factor=&VectorFromFloat(Value=&FloatFromCurve(
      FloatCurve=DataInterface() {propertyValues {Curve {keys [
          {interpMode=RCIM_Cubic, tangentMode=RCTM_User, time=0, value=0}
          {interpMode=RCIM_Cubic, tangentMode=RCTM_User, time=0.2, value=1}
          {interpMode=RCIM_Cubic, tangentMode=RCTM_User, time=0.5, value=1}
          {interpMode=RCIM_Cubic, tangentMode=RCTM_User, time=1, value=0}
         ]}}}
      Scale Curve=&RandomRangeFloat(Minimum=1.2, Maximum=1.5)
     )))
   &CameraOffset(Camera Offset Amount=&FN_FrontEndFloatModifier(
     Gameplay=&RandomRangeFloat(Minimum=5, Maximum=4)
     Front End=&RandomRangeFloat(Minimum=1, Maximum=1.5)
    ))
  ]
  rendererClasses [&NiagaraMeshRendererProperties]
  renderers [{rendererClass=&NiagaraMeshRendererProperties, rendererIndex=0}]
 }
 Drops_World {
  emitterUpdateScript [
   &EmitterState(Life Cycle Mode=::NewEnumerator1, Loop Duration=1, UseLoopDelay=false)
   &SpawnPerUnit(
    Max Movement Threshold=600
    Module.Use Max Movement Threshold=true
    Module.Use Movement Tolerance=false
    Spawn Spacing=25
    Use Spawn Probability=false
   )
  ]
  particleSpawnScript [
   &InitializeParticle(
    Color=&RandomRangeLinearColor(Minimum=<rgba 0.2917 0.15 1 0.9>, Maximum=<rgba 0.7167 0.15 1 0.9>)
    Lifetime Max=0.5
    Lifetime Min=0.3
    Lifetime Mode=::NewEnumerator1
    Mesh Scale Mode=::NewEnumerator1
    Mesh Uniform Scale Max=0.2
    Mesh Uniform Scale Min=0.8
    Sprite Rotation Angle=&RandomRangeFloat(Minimum=0, Maximum=1)
    Sprite Rotation Mode=::NewEnumerator1
    UsePositionOffset=false
    Write Material Random=true
   )
   &SkeletalMeshLocation(
    Apply To Particle Position=false
    Bone / Socket Index=&SpawnGroup()
    Inherit Velocity=true
    "Inherit Velocity (Scale)"=0.02
    MaskSpawnGroup=false
    Mesh Sampling Type=::NewEnumerator5
    Sampled Position Offset=&RandomRangeVector(Minimum=<0 10 -10>, Maximum=<0 -10 20>)
    Sampled Position Offset Coordinate Space:ENiagaraMeshTransforms=::NewEnumerator3
    Skeletal Mesh=$User.SkeletalMesh
    Socket Sampling Mode=::NewEnumerator5
    VelocitySpeedLimit=false
   ) {module=SkeletalMeshLocation001}
   &CylinderLocation(
    Cylinder Axis:Vector3f=<-1 0 0>
    Cylinder Height=1
    Cylinder Radius=10
    MaskSpawnGroup=false
    Non Uniform Scale:Vector3f=<1.5 1 1>
    Offset=<-5 0 0>
    Offset Coordinate Space:ENiagaraCoordinateSpace=::Local
    Offset Mode:ENiagara_OffsetMode=::NewEnumerator0
    Override Seed=false
    Rotation Coordinate Space:ENiagaraCoordinateSpace=::Simulation
    Rotation Mode:ENiagara_RotationMode=::NewEnumerator1
    Surface Only=true
    Use Endcaps In Surface Only Mode=false
    UseOwnerScale=true
   )
  ]
  particleUpdateScript [
   &ParticleState()
   &AccelerationForce(Acceleration=&RandomRangeVector(Minimum=<0 0 -500>, Maximum=<0 0 -200>))
   &InheritSourceMovement(
    Applied Position Falloff Distance=50
    Applied Position Scale=<0.8 0.8 0.8>
    ApplyForce=false
    ApplyPosition=true
    ApplyPositionFalloff=true
    ApplyPositionFalloffCurve=false
   )
   &VectorNoiseForce(Force Amount=&RandomRangeFloat(Minimum=250, Maximum=150), Module.Use Random Noise Time Interval=false)
   &Drag(Drag=5, Use Linear Drag=true, Use Rotational Drag=false)
   &SolveForcesAndVelocity(Clamp Velocity=false, Limit Acceleration=false)
   &GenerateLocationEvent(Module.Use Event Send Rate=true)
   &ScaleMeshSize(Scale Factor=&VectorFromFloat(Value=&FloatFromCurve(
      FloatCurve=DataInterface() {propertyValues {Curve {keys [
          {interpMode=RCIM_Cubic, tangentMode=RCTM_User, time=0, value=0}
          {interpMode=RCIM_Cubic, tangentMode=RCTM_User, time=0.2, value=1}
          {interpMode=RCIM_Cubic, tangentMode=RCTM_User, time=0.5, value=1}
          {interpMode=RCIM_Cubic, tangentMode=RCTM_User, time=1, value=0}
         ]}}}
      Scale Curve=&RandomRangeFloat(Minimum=1.2, Maximum=1.8)
     )))
  ]
  rendererClasses [&NiagaraMeshRendererProperties]
  renderers [{rendererClass=&NiagaraMeshRendererProperties, rendererIndex=0}]
 }
}
summary {
 systemName=NS_Pickaxe_Cereal_Box_Idle_Cosmos
 emitters [
  {emitterName=Drops, rendererClasses [&NiagaraMeshRendererProperties]}
  {emitterName=Drops_Small, rendererClasses [&NiagaraMeshRendererProperties]}
  {emitterName=Drops_World, rendererClasses [&NiagaraMeshRendererProperties]}
 ]
 userVariables [
  User.bIsFrontEnd {type=NiagaraBool}
  User.bIsFrontEndPreview {type=NiagaraBool}
  User.RibbonIDOffset {type=NiagaraInt32}
  User.SkeletalMesh {
   type=NiagaraDataInterfaceSkeletalMesh
   defaultValue {
    struct=NiagaraExt_VariableValue_DataInterface
    value {dataInterface=&self:NiagaraDataInterfaceSkeletalMesh_0}
   }
  }
 ]
}
